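# Material function for the SatireCane backpack that adds a timestamp-triggered emissive reaction to incoming material attributes: animated outlines, then a glow.
# A packed ReactivityMask supplies outline shape, outline gradient and inner shape.
# Frac of time and gradient forms repeating travelling bands along the outlines, with tunable speed, tiling and length.
# The glow timeline starts when the outline duration ends; both stages tint emissive by base color and layer through attribute blends.
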MaterialFunction /BRCosmetics/Accessories/FORT_Backpacks/Backpack_M_MED_SatireCane_Ode/Materials/MF_SatireCane_Reactivity
graph {nodes [
  n1 = NamedRerouteDeclaration(None=FunctionInput() {inputName=Material, inputType=FunctionInput_MaterialAttributes}) {name=IN_MA, nodeColor=<rgba 1 0.9375 0 1>}
  n2 = GetMaterialAttributes(None=@n1) {attributeGetTypes [#1, #2]}
  n3 = Multiply(
   A=Multiply(A=@n2.BaseColor, B=VectorParameter() {group=Reactivity, parameterName=OutlineColor, defaultValue=<rgba 1 1 1 1>}.RGB)
   B=ScalarParameter() {defaultValue=1, group=Reactivity, parameterName=OutlineEmissiveIntensity}
  )
  n4 = BlendMaterialAttributes(
   A=@n1
   B=SetMaterialAttributes!(MaterialAttributes=@n2.MaterialAttributes, EmissiveColor=Add(A=@n2.EmissiveColor, B=@n3)) {attributeSetTypes [#1]}
   Alpha=NamedRerouteUsage()
  )
  n5 = GetMaterialAttributes(None=@n4) {attributeGetTypes [#1, #2]}
  n6 = Max(
   A=Multiply(A=VectorParameter() {group=Reactivity, parameterName=GlowColor, defaultValue=<rgba 1 1 1 1>}.RGB, B=NamedRerouteUsage())
   B=Multiply(A=VectorParameter() {group=Reactivity, parameterName=OutlineGlowColor, defaultValue=<rgba 1 1 1 1>}.RGB, B=NamedRerouteUsage())
  )
  n7 = SetMaterialAttributes(MaterialAttributes=@n5.MaterialAttributes, EmissiveColor=Add(A=@n5.EmissiveColor, B=Multiply(A=@n6, B=@n5.BaseColor))) {attributeSetTypes [#1]}
  n8 = NamedRerouteUsage()
  FunctionOutput(None=NamedRerouteDeclaration(None=BlendMaterialAttributes(A=@n4, B=@n7, Alpha=@n8)) {name=OUT_MA, nodeColor=<rgba 0.6172 0 1 1>})
  n9 = TextureSampleParameter2D() {group=Reactivity, parameterName=ReactivityMask, samplerType=SAMPLERTYPE_Masks, texture=&T_BlackMask}
  NamedRerouteDeclaration(None=@n9.R) {name=Outline_Shape, nodeColor=<rgba 0.4766 0 1 1>}
  NamedRerouteDeclaration(None=@n9.G) {name=Outline_Gradient, nodeColor=<rgba 0 1 0.1562 1>}
  NamedRerouteDeclaration(None=@n9.B) {name=Inner_Shape, nodeColor=<rgba 1 0 0.1641 1>}
  n10 = NamedRerouteDeclaration(None=MaterialFunctionCall!(
    TimeStart=ScalarParameter() {group=Reactivity, parameterName=Reactivity_Timestamp}
    AnimLength=ScalarParameter() {defaultValue=1, group=Reactivity, parameterName=Reactivity_OutlineDuration}
    TimeCurrent=Time()
   ) {materialFunction=/Game/Global/MaterialFunctions/TechArt/MF_TimeStampAnim}.AnimValue) {name=Timeline_Outlines, nodeColor=<rgba 1 0.1406 0 1>}
  n11 = Step(Y=@n10) {constX=0.01}
  n12 = Step(Y=@n10) {constX=0.99}
  n13 = ScalarParameter() {defaultValue=0.25, group=Reactivity, parameterName=Reactivity_FadeInDuration}
  NamedRerouteDeclaration(None=Multiply(A=Add(A=@n11, B=@n12), B=SmoothStep(Max=@n13, Value=@n10) {constMax=0.25})) {name=Timeline_Outlines_Alpha, nodeColor=<rgba 1 0.1172 0 1>}
  n14 = Add(
   A=ScalarParameter() {group=Reactivity, parameterName=Reactivity_Timestamp}
   B=ScalarParameter() {defaultValue=1, group=Reactivity, parameterName=Reactivity_OutlineDuration}
  )
  n15 = MaterialFunctionCall(
   TimestampA=@n14
   TransitionLength_In=Constant() {r=0.1}
   TimestampB=Add(A=@n14, B=ScalarParameter() {defaultValue=1, group=Reactivity, parameterName=Reactivity_GlowDuration})
   TransitionLength_Out=ScalarParameter() {defaultValue=1, group=Reactivity, parameterName=Reactivity_GlowFadeOut}
   Time=Time()
  ) {materialFunction=&MF_Timestamp_InOut_01}
  NamedRerouteDeclaration(None=Multiply(A=@n15."SmoothStep (A Only)", B=@n15."SmoothStep (A Only)")) {name=TImeline_Glow, nodeColor=<rgba 0.2891 0 1 1>}
  n16 = NamedRerouteUsage()
  n17 = ScalarParameter() {defaultValue=3, group=Reactivity, parameterName=OutlineSpeed}
  n18 = Subtract(
   A=Frac(None=Multiply(A=@n16, B=@n17) {constB=3})
   B=Multiply(A=NamedRerouteUsage(), B=ScalarParameter() {defaultValue=0.75, group=Reactivity, parameterName=OutlineTiling}) {constB=10}
  )
  n19 = Saturate(None=Divide(A=Frac(None=@n18), B=ScalarParameter() {defaultValue=0.4, group=Reactivity, parameterName=OutlineLength}) {constB=0.2})
  n20 = Step(Y=NamedRerouteUsage()) {constX=0}
  n21 = Multiply(A=ConstantBiasScale(None=OneMinus(None=@n19)) {bias=-0.5, scale=2}, B=Multiply(A=NamedRerouteUsage(), B=OneMinus(None=@n20)))
  n22 = NamedRerouteUsage()
  NamedRerouteDeclaration(None=Saturate(None=Multiply(A=@n21, B=@n22))) {name=OUT_Outlines, nodeColor=<rgba 0 1 0.2969 1>}
 ]}
